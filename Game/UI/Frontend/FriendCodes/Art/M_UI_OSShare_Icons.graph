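# M_UI_OSShare_Icons is a translucent UI material for platform-specific share icons in the frontend friend code UI.
# Static switch bIsDefaultPlatform picks between the FriendLink icon's alpha and a channel of the packed OS share icon texture.
# A nested bIsApple_Color switch feeds the texture's blue channel into both branches, so it currently makes no difference.
# A constant of 1 accompanies the mask; material output connections are not shown.
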Material /Game/UI/Frontend/FriendCodes/Art/M_UI_OSShare_Icons
graph {
 properties {
  bUsedWithStaticMesh=false
  blendMode=BLEND_Translucent
  materialDomain=MD_UI
  pixelDepthOffsetMode=PDOM_Legacy
  usedShadingModels=Default Lit
 }
 nodes [
  Constant() {r=1}
  n1 = TextureSample() {texture=&/T_UI_Icons_OSShare}
  StaticSwitchParameter(
   True=TextureSample() {texture=/Game/UI/SocialPanel/Art/Textures/T_UI_Icon_FriendLink_64}.A
   False=StaticSwitchParameter(True=@n1.B, False=@n1.B) {parameterName=bIsApple_Color}
  ) {parameterName=bIsDefaultPlatform}
 ]
}
